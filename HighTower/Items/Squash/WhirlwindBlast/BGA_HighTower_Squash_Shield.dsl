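; Damageable shield actor for HighTower Squash, filed under the Whirlwind Blast item, with a short failsafe lifespan.
; On begin play it halts the instigating pawn's movement and forces a net update.
; Server EndShield hides it after delays and silently kills it; death before ending marks it destroyed and hides it.
; OnRep_Hide toggles visibility and collision, while damage fires a replicated gameplay cue at the impact, gated by DoOnce against spam.
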(event Custom|SetFailsafeLifeSpan
  ;[1 1 1] Failsafe Lifespan.
  (Actor|SetLifeSpan 3.0))

(event EventBeginPlay
  ;[1 1 1] Begin Play - Play GC
  (CallFunction|SetFailsafeLifeSpan)
  (Networking|SwitchHasAuthority
    (:Authority
      (Variables|Debug|SetThicknessMult 0.25))
    (:Remote
      _))
  (Variables|Default|SetOneOverSphereRadius (/ 1.0 (Variables|Default|GetSphereRadius)))
  (bind _asfort_player_pawn (Utilities|Casting|CastToFortPlayerPawn (Game|GetInstigator)))
  (Components|Movement|StopMovementImmediately (Variables|/Script/Engine.Character|GetCharacterMovement _asfort_player_pawn))
  (Networking|ForceNetUpdate _asfort_player_pawn))

(event Custom|EndShield
  (params (StartEndFXOnly :type Boolean))
  ;[1 1 1] End Shield
  (if (Networking|IsServer)
    (if (not (Variables|Default|GetDestroyed))
      (Utilities|FlowControl|DoOnce false
        (:Completed
          (Utilities|FlowControl|Delay 0.5)
          (Variables|Default|SetEnded true)
          (Variables|Default|SetHide true)
          (Utilities|FlowControl|Delay (Variables|Default|GetDelay_EndBeforeDeath))
          (Gameplay|SilentDie false))))))

(event EventOnDeathServer
  (params (Damage :type Float%20%28single-precision%29) (DamageTags :type Gameplay%20Tag%20Container%20Structure%20%28by%20ref%29) (Momentum :type Vector :default "0, 0, 0") (HitInfo :type Hit%20Result%20Structure%20%28by%20ref%29) (InstigatedBy :type Controller%20Object%20Reference) (DamageCauser :type Actor%20Object%20Reference) (EffectContext :type Gameplay%20Effect%20Context%20Handle%20Structure))
  ;[1 1 1] On Death Server - If the shield dies check if we have already ended if not play death GC and set the notify
  (if (not (Variables|Default|GetEnded))
    (Variables|Default|SetDestroyed true)
    (Development|PrintString "shield Destroyed" (Variables|Debug|GetDrawDebug) (Variables|Debug|GetDrawDebug))
    (Variables|Default|SetHide true)))

(event EventOnDamageServer
  (params (Damage :type Float%20%28single-precision%29) (DamageTags :type Gameplay%20Tag%20Container%20Structure%20%28by%20ref%29) (Momentum :type Vector :default "0, 0, 0") (HitInfo :type Hit%20Result%20Structure%20%28by%20ref%29) (InstigatedBy :type Controller%20Object%20Reference) (DamageCauser :type Actor%20Object%20Reference) (EffectContext :type Gameplay%20Effect%20Context%20Handle%20Structure))
  (bind (_bblockinghit _binitialoverlap _time _distance _location _impactpoint _normal) (Collision|BreakHitResult HitInfo))
  ;[1 1 1] On Damage Server - when damage is taken play a GC at the location. After do a delay so it does not spam too much with super fast weapons.
  (Utilities|FlowControl|DoOnce false
    (:Completed
      (Ability|ExecuteGameplayCueWithParamsReplicated (Variables|GamplayCues|GetGC_Damaged) (Ability|GameplayCue|MakeGameplayCueParameters :NormalizedMagnitude 0.0 :RawMagnitude 0.0 :Location _impactpoint :Normal _normal :GameplayEffectLevel 1 :AbilityLevel 1 :bReplicateLocationWhenUsingMinimalRepProxy false))
      (Utilities|FlowControl|Delay (Variables|GamplayCues|GetDelay_DamageGC)) ; Dont allow the Damage GC to trigger too often
    )))

(fn OnRep_Hide
  (bind _hide (Variables|Default|GetHide))
  (Rendering|SetActorHiddenInGame _hide)
  (Collision|SetActorEnableCollision (not _hide)))

(fn ConstructionScript)
