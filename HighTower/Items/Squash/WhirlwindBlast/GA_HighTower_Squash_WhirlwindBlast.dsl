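; Squash Whirlwind Blast gameplay ability for HighTower: a low-gravity blast with an attached shield and projectile deflection.
; Activation applies low gravity, changes movement mode and starts a looping sphere-overlap timer that reverses projectile velocities with throttled deflect cues.
; Montage notifies step through indexed targeting tags for knockback, starting the end timer and reporting players hit to quests.
; Shield death triggers net-synced cleanup; ability end removes low gravity and restores falling.
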(event Custom|ResetAbility
  ;[1 1 1] initializeTheNotifies
  (Variables|PowerTargetApply|SetTag_TargetingApplication "(TagName=\"Targeting.ApplicationTag.0\")")
  (Variables|Default|SetNotifyIndex 0)
  (Variables|Default|SetGCIndex 0)
  (Utilities|Time|ClearandInvalidateTimerbyHandle (Variables|Default|GetEndAbilityTimer))
  (CallFunction|ResetCanbeDestroyedTimedOutorCancelled)
  (Variables|Default|SetDestroyed false)
  (Variables|Default|SetEndAbilityHasBeenCalled false)
  (Utilities|Array|Clear (Variables|Default|GetTargetsHit))
  (Utilities|Array|Clear (Variables|Default|GetTargetCache)))

(event EventOnAbilityActivated
  (bind _playerpawn (Variables|Default|GetPlayerPawn))
  (bind _charactermovement (Variables|/Script/Engine.Character|GetCharacterMovement _playerpawn))
  ;[1 1 1] Set up timer to check for projectiles during the ability
  (CallFunction|ResetAbility)
  (bind _returnvalue (Utilities|Time|SetTimerbyEvent (delegate Custom|FindProjectiles) (Variables|Default|GetProjectileCheckInterval) true))
  (Variables|Default|SetTimer_ProjectileCheck _returnvalue)
  (|CallParentFunction)
  (bind _returnvalue_2 (Ability|ApplyGameplayEffectToOwner "/HighTower/Items/Squash/WhirlwindBlast/GE_HighTower_Squash_WhirldwindBlast_LowGrav.GE_HighTower_Squash_WhirldwindBlast_LowGrav_C"))
  (Variables|Default|SetGEHandle_LowGrav _returnvalue_2)
  (if (== (Variables|Default|GetWhichMovementMode) 0)
    (Pawn|Components|CharacterMovement|SetMovementMode _charactermovement "MOVE_Custom" (Variables|Default|GetMovementModeHover))
    (elif (== (Variables|Default|GetWhichMovementMode) 1)
      (Pawn|Components|CharacterMovement|SetMovementMode _charactermovement "MOVE_Custom" (Variables|Default|GetMovementModeFloating))
      (else
        (Pawn|Components|CharacterMovement|SetMovementMode _charactermovement "MOVE_Falling" 0)
        (CallFunction|ExecuteandIncrementKnock true)
        (CallFunction|SpawnShield)
        (if (not (Variables|Default|GetDebugAbility))
          (Utilities|PowerCommitAbility false))))))

(event Custom|FindProjectiles
  ;[1 1 1] Get overlapping projectiles, call the reflect function then add them to the ignore array
  (bind (_outactors _returnvalue) (Collision|SphereOverlapActors (Transformation|GetActorLocation (Variables|Default|GetPlayerPawn)) 6e+02 (Utilities|Array|MakeArray "ObjectTypeQuery8" "ObjectTypeQuery10") 0 (Variables|Default|GetProjectilesReflected)))
  (if _returnvalue
    (for _array_element _outactors
      (CallFunction|ReflectProjectile _array_element)
      (Utilities|Array|AddUnique (Variables|Default|GetProjectilesReflected) _array_element))))

(event Custom|ReflectProjectile
  (params (Projectile :type Actor%20Object%20Reference%20%28by%20ref%29))
  (bind _returnvalue (Actor|GetComponentbyClass Projectile "/Script/Engine.ProjectileMovementComponent"))
  (bind _returnvalue_2 (Actor|GetComponentbyClass Projectile "/Script/FortniteGame.FortProjectileMovementComponent"))
  ;[1 1 1] Projectiles - if valid movement component, invert the velocity.
  (CallFunction|OnDeflection (Transformation|GetActorLocation Projectile))
  (Utilities|IsValid _returnvalue
    (:Is%20Valid
      (Variables|/Script/Engine.MovementComponent|SetVelocity (* (Variables|/Script/Engine.MovementComponent|GetVelocity _returnvalue) -1.0) _returnvalue))
    (:Is%20Not%20Valid
      (Utilities|IsValid _returnvalue_2
        (:Is%20Valid
          (Projectile|SetVelocity _returnvalue_2 (* (Variables|/Script/Engine.MovementComponent|GetVelocity _returnvalue_2) -1.0)))))))

(event Custom|SpawnShield
  ;[1 1 1] Spawn Shield - Spawn the shield and attach it to the player
  (if (Networking|IsServer)
    (Utilities|IsValid (Variables|Default|GetShield)
      (:Is%20Valid
        (CallFunction|OnShieldDestroyedTimedOutorCancelled))
      (:Is%20Not%20Valid
        (bind _returnvalue (Game|SpawnActorfromClass "/HighTower/Items/Squash/WhirlwindBlast/BGA_HighTower_Squash_Shield.BGA_HighTower_Squash_Shield_C" (Math|Transform|MakeTransform (Transformation|GetActorLocation (Variables|Default|GetPlayerPawn)) (Transformation|GetActorRotation (Variables|Default|GetPlayerPawn))) "AlwaysSpawn" "OverrideRootScale" 0 0 (Variables|Default|GetPlayerPawn)))
        (bind _output_get (Variables|Default|SetShield _returnvalue))
        (EventDispatchers|BindEventtoOnDied _output_get (EventDispatchers|CreateEvent :!event Custom|ShieldDeath))
        (Transformation|AttachActorToActor _returnvalue (Variables|Default|GetPlayerPawn) "None" "SnapToTarget" "KeepWorld" "KeepRelative" true)
        (CallFunction|SetupBinds)))))

(event Custom|SetupBinds
  ;[1 1 1] Set up additional Binds for ending the ability
  (Ability|Tasks|WaitGameplayTagAdd "(TagName=\"Gameplay.Action.Player.DanceStunned\")"
    (:Added
      (Ability|EndAbility))))

(event Custom|OnDeflection
  (params (HitLoc :type Vector))
  ;[1 1 1] On Deflection - Play a GC with delay to prevent spam.
  (Utilities|FlowControl|DoOnce false
    (:Completed
      (Ability|ExecuteGameplayCueWithParamsOnOwner (Variables|Default|GetGCDeflect) (Ability|GameplayCue|MakeGameplayCueParameters :NormalizedMagnitude 0.0 :RawMagnitude 0.0 :Location HitLoc :Normal (Math|Vector|Normalize (- HitLoc (Transformation|GetActorLocation (Variables|Default|GetPlayerPawn)))) :GameplayEffectLevel 1 :AbilityLevel 1 :bReplicateLocationWhenUsingMinimalRepProxy false))
      (Utilities|FlowControl|Delay (Variables|Default|GetDelayDeflectGC)) ; Dont allow the Damage GC to trigger too often
    )))

(event Custom|EndCalledCleanUp
  (bind _shield (Variables|Default|GetShield))
  ;[1 1 1] On shield destroyed or cancelled
  (Development|PrintString "On Shield Cancelled Timed Out Or Destroyed" (Variables|Default|GetDebugAbility) (Variables|Default|GetDebugAbility))
  (if (not (Variables|Default|GetDestroyed))
    (Utilities|IsValid _shield
      (:Is%20Valid
        (Class|BGAHighTowerSquashShield|EndShield _shield false))
      (:Is%20Not%20Valid
        _))
    (if (Variables|Default|GetEndAbilityHasBeenCalled)
      (if (not (Variables|Default|GetEndAbilityHasBeenCalled))
        ;;[1 1 1] Delay and end
        (Utilities|FlowControl|Delay (Variables|Default|GetPostEndCleanUpDelay)) ; Wait a moment before ending so that everything has time to resolve correctly
        (Ability|EndAbility)))))

(event Custom|ShieldDestoyedClient
  (bind _shield (Variables|Default|GetShield))
  ;[1 1 1] On shield destroyed or cancelled
  (Utilities|FlowControl|DoOnce false
    (:Completed
      (Ability|Tasks|WaitNetSync "BothWait"
        (:OnSync
          (Development|PrintString "On Shield Cancelled Timed Out Or Destroyed" (Variables|Default|GetDebugAbility) (Variables|Default|GetDebugAbility))
          (if (not (Variables|Default|GetDestroyed))
            (Utilities|IsValid _shield
              (:Is%20Valid
                (Class|BGAHighTowerSquashShield|EndShield _shield false))
              (:Is%20Not%20Valid
                _))
            (if (Variables|Default|GetEndAbilityHasBeenCalled)
              (if (not (Variables|Default|GetEndAbilityHasBeenCalled))
                ;;[1 1 1] Delay and end
                (Utilities|FlowControl|Delay (Variables|Default|GetPostEndCleanUpDelay)) ; Wait a moment before ending so that everything has time to resolve correctly
                (Ability|EndAbility)))))) ; Ensure the client and server sync up here.
    )))

(event Custom|OnShieldDestroyedTimedOutOrCancelled
  (bind _shield (Variables|Default|GetShield))
  ;[1 1 1] On shield destroyed or cancelled
  (Utilities|FlowControl|DoOnce false
    (:Completed
      (Ability|Tasks|WaitNetSync "BothWait"
        (:OnSync
          (Development|PrintString "On Shield Cancelled Timed Out Or Destroyed" (Variables|Default|GetDebugAbility) (Variables|Default|GetDebugAbility))
          (if (not (Variables|Default|GetDestroyed))
            (Utilities|IsValid _shield
              (:Is%20Valid
                (Class|BGAHighTowerSquashShield|EndShield _shield false))
              (:Is%20Not%20Valid
                _))
            (if (Variables|Default|GetEndAbilityHasBeenCalled)
              (if (not (Variables|Default|GetEndAbilityHasBeenCalled))
                ;;[1 1 1] Delay and end
                (Utilities|FlowControl|Delay (Variables|Default|GetPostEndCleanUpDelay)) ; Wait a moment before ending so that everything has time to resolve correctly
                (Ability|EndAbility)))))) ; Ensure the client and server sync up here.
    )))

(event Custom|ResetCanBeDestroyedTimedOutOrCancelled
  (bind _shield (Variables|Default|GetShield))
  ;[1 1 1] On shield destroyed or cancelled
  (Utilities|FlowControl|DoOnce false
    (:Completed
      (Ability|Tasks|WaitNetSync "BothWait"
        (:OnSync
          (Development|PrintString "On Shield Cancelled Timed Out Or Destroyed" (Variables|Default|GetDebugAbility) (Variables|Default|GetDebugAbility))
          (if (not (Variables|Default|GetDestroyed))
            (Utilities|IsValid _shield
              (:Is%20Valid
                (Class|BGAHighTowerSquashShield|EndShield _shield false))
              (:Is%20Not%20Valid
                _))
            (if (Variables|Default|GetEndAbilityHasBeenCalled)
              (if (not (Variables|Default|GetEndAbilityHasBeenCalled))
                ;;[1 1 1] Delay and end
                (Utilities|FlowControl|Delay (Variables|Default|GetPostEndCleanUpDelay)) ; Wait a moment before ending so that everything has time to resolve correctly
                (Ability|EndAbility)))))) ; Ensure the client and server sync up here.
    )))

(event EventCastMontageNotifyStart
  (params (NotifyTag :type Gameplay%20Tag%20Structure))
  (bind _notifyindex (Variables|Default|GetNotifyIndex))
  ;[1 1 1] Blast On AnimNotify
  (if (Variables|Default|GetDebugAbility)
    ;[1 1 1] print notifies
    (Development|PrintString (Utilities|String|Append "NOTIFY:" (Utilities|String|ToString(Name) (GameplayTags|GetTagName NotifyTag)) "     GCIndex:" (Utilities|String|ToString(Integer) (Variables|Default|GetNotifyIndex))) true true "(R=0.000000,G=0.660000,B=1.000000,A=1.000000)" 1e+01))
  (if (< _notifyindex (Utilities|Array|Length (Variables|Default|GetTag_TargetingApplication_Blast)))
    (Utilities|FlowControl|Sequence
      (:then_0
        (if (> (Variables|Default|GetNotifyIndex) 0)
          (CallFunction|ExecuteandIncrementKnock false)))
      (:then_1
        (Math|Integer|IncrementInt _notifyindex)))
    (else
      (|CallParentFunction NotifyTag))))

(event Custom|ShieldDeath
  (params (DamagedActor :type Actor%20Object%20Reference) (Damage :type Float%20%28single-precision%29) (InstigatedBy :type Controller%20Object%20Reference) (DamageCauser :type Actor%20Object%20Reference) (HitLocation :type Vector) (FHitComponent :type Primitive%20Component%20Object%20Reference) (BoneName :type Name) (Momentum :type Vector))
  ;[1 1 1] Bramble Death - End Ability if shield dies
  (Development|PrintString "Shield Death (Ability)" (Variables|Default|GetDebugLogShield) (Variables|Default|GetDebugLogShield) "(R=0.000000,G=0.660000,B=1.000000,A=1.000000)" 1e+01)
  (Variables|Default|SetDestroyed true)
  (CallFunction|OnShieldDestroyedTimedOutorCancelled)
  (CallFunction|ShieldDestoyedClient))

(event EventCastMontageBlendOut)

(event EventCastMontageCancelledOrInterrupted)

(event Ability|EventOnEndAbility
  (params (bWasCancelled :type Boolean))
  ;[1 1 1] Clear the Projectile Checking Timer on End Ability
  (Ability|EndAbilityState "None")
  (Utilities|Time|ClearandInvalidateTimerbyHandle (Variables|Default|GetTimer_ProjectileCheck))
  (|CallParentFunction bWasCancelled)
  (Ability|RemoveGameplayEffectFromOwnerWithHandle (Variables|Default|GetGEHandle_LowGrav))
  (Pawn|Components|CharacterMovement|SetMovementMode (Variables|/Script/Engine.Character|GetCharacterMovement (Variables|Default|GetPlayerPawn)) "MOVE_Falling" 0)
  (Variables|Default|SetEndAbilityHasBeenCalled true)
  (CallFunction|EndCalledCleanUp))

(event Custom|ExecuteAndIncrementKnock
  (params (IsFirstKnock :type Boolean))
  (bind _gcindex (Variables|Default|GetGCIndex))
  ;[1 1 1] Execute and increment knock
  (if IsFirstKnock
    (Variables|Default|SetGCIndex 0)
    (else
      _))
  (Utilities|FlowControl|Sequence
    (:then_0
      (bind (_asynctaskproxy _targetdata _applicationtag) (Ability|Tasks|DoTargetSelection "Targeting" (Utilities|Array|Get(aref) (Variables|Default|GetTag_TargetingApplication_Blast) _gcindex) true)
        (:then
          (if (< _gcindex (Utilities|Array|Length (Variables|Default|GetTag_TargetingApplication_Blast)))
            (if (== (Variables|Default|GetGCIndex) 0)
              (Utilities|IsServer
                (:false
                  (CallFunction|PlayBurstGC (Variables|PowerGameplayCues|GetGC_Activation) (Variables|PowerGameplayCues|GetActivationGCUsesParams) (CallFunction|ActivationGCParams))))
              (else
                (CallFunction|PlayBurstGC (Variables|Default|GetKnockBackGC) (Variables|PowerGameplayCues|GetActivationGCUsesParams) (CallFunction|ActivationGCParams))
                (bind _returnvalue (Utilities|Time|SetTimerbyEvent (delegate Custom|Timeout) (Variables|Default|GetPostKnockSeconds)))
                (Variables|Default|SetEndAbilityTimer _returnvalue)
                ;[1 1 1] Count players hit and send quest progress
                (if (Networking|IsServer)
                  (Variables|Default|SetQuestProgressCount 0)
                  (for _array_element (Variables|Default|GetTargetsHit)
                    (Utilities|IsValid _array_element
                      (:Is%20Valid
                        (Utilities|Casting|CastToFortPlayerPawn _array_element)
                        (Math|Integer|IncrementInt (Variables|Default|GetQuestProgressCount)))))
                  (Quest|SendComplexCustomStatEvent (Quests|GetQuestManager (Fortnite|GetFortPlayerControllerfromActor (Variables|Default|GetPlayerPawn)) "Invalid") 0 "(GameplayTags=((TagName=\"Athena.Quests.HighTowerSquashHit\")))" "(GameplayTags=)" (Variables|Default|GetQuestProgressCount)))))))
        (:Targeted
          (CallFunction|ApplyEffect _targetdata _applicationtag)
          ;;[1 1 1] Build array of targets
          (bind _output_get (Variables|Default|SetTargetCache (Ability|TargetData|GetAllActorsfromTargetData _targetdata)))
          (for _array_element_2 _output_get
            (Utilities|Array|AddUnique (Variables|Default|GetTargetsHit) _array_element_2)))))
    (:then_1
      (Math|Integer|IncrementInt _gcindex))))

(event Custom|Timeout
  ;[1 1 1] Execute and increment knock
  (Ability|EndAbility))

(fn InitAbility
  ;[1 1 1] Clear Projectile array on activation
  (Utilities|Array|Clear (Variables|Default|GetProjectilesReflected))
  (|CallParentFunction))
