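; Gameplay ability for the initial knock-up of the Squash item's Whirlwind Blast in HighTower.
; On activation it runs the parent logic, then applies the InitialKnockLowGrav gameplay effect to the owner, which by its name lowers gravity during the knock.
; Ending only forwards to the parent, so any launch, effect removal or ability termination is not handled in this graph.
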(event Ability|EventActivateAbility
  (|CallParentFunction)
  (Ability|ApplyGameplayEffectToOwner "/HighTower/Items/Squash/WhirlwindBlast/GE_HighTower_Squash_WhirldwindBlast_InitialKnockLowGrav.GE_HighTower_Squash_WhirldwindBlast_InitialKnockLowGrav_C"))

(event Ability|EventOnEndAbility
  (params (bWasCancelled :type Boolean))
  (|CallParentFunction bWasCancelled))
