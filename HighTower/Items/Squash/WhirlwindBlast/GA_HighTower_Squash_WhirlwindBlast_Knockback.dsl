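; Whirlwind Blast knockback gameplay ability for the HighTower Squash item, extending a knockback parent ability with quest tracking.
; On activation it runs the parent logic, then applies a quest gameplay effect to the knockback target's ability system component.
; The effect tags knocked-back enemies so damage dealt to them counts toward a Season 14 Week 7 quest for Storm's Whirlwind Blast.
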(event Ability|EventActivateAbility
  (|CallParentFunction)
  (GameplayEffects|ApplyGameplayEffectToSelf :self (Ability|GetAbilitySystemComponent (Variables|KnockbackParent|GetTarget)) :GameplayEffectClass "/HighTower/Items/Squash/WhirlwindBlast/GE_HighTower_Squash_Quest.GE_HighTower_Squash_Quest_C" :Level 0.0) ; Season 14 - Week 7 - Quest 7\nDeal Damage after knocking an opponent back with Storm's Whirlwind Blast\nApply tag to knocked-back enemies so damage to them can be tracked
)
